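; Basic_StairW is an actor Blueprint in the InvisibleGeo environment folder, apparently a stair piece, whose only logic cleans up child actors.
; Its OnDestroyed_Event handler gets all child actors of the destroyed actor, including descendants, and destroys them.
; The OnDestroyed binding sits inside that same handler, and no other event here makes the binding.
; The construction script is empty.
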(event Custom|OnDestroyed_Event
  (params (DestroyedActor :type Actor%20Object%20Reference))
  (Game|BindEventtoOnDestroyed (delegate Custom|OnDestroyed_Event))
  (Actor|DestroyActor (Actor|GetAllChildActors DestroyedActor true)))

(fn ConstructionScript)
